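# Control Rig for the Vivid Splash pickaxe that spins the paint roller bone at speeds driven by character state.
# Sequential idle, movement, zipline, falling, front-end and melee checks set a speed/duration vector, later ones overriding earlier.
# Melee builds a swing counter that tunes duration and blend time; front end alternates stillness with random speeds.
# A timed gate latches the speed, which an accumulated lerp smooths into a looping Z rotation.
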ControlRigBlueprint /BRCosmetics/Weapons/FORT_Melee/Pickaxe_VividSplash/Meshes/VividSplash_Axe_CtrlRig
graphs [
 RigVMFunctionLibrary {links [], nodes []}
 RigVMModel {nodes [
   RigUnit_BeginExecution = &RigUnit_BeginExecution()
   RigVMFunction_ControlFlowBranch = &RigVMFunction_ControlFlowBranch(ExecuteContext=@RigUnit_BeginExecution.ExecutePin, Condition=RigVMVariableNode() {variable=Is Idle}.Value)
   n1 = &RigVMFunction_ControlFlowBranch(ExecuteContext=@RigVMFunction_ControlFlowBranch.Completed, Condition=RigVMVariableNode() {variable=IsMoving2D}.Value)
   n2 = &RigVMFunction_ControlFlowBranch(ExecuteContext=@n1.Completed, Condition=RigVMVariableNode() {variable=Is Zipplining}.Value)
   n3 = &RigVMFunction_ControlFlowBranch(ExecuteContext=@n2.Completed, Condition=RigVMVariableNode() {variable=Is Falling}.Value)
   n4 = &RigVMFunction_ControlFlowBranch(ExecuteContext=@n3.Completed, Condition=RigVMVariableNode() {variable=IsFrontEnd}.Value)
   n5 = &RigVMFunction_ControlFlowBranch(ExecuteContext=@n4.Completed, Condition=RigVMVariableNode() {variable=Is Melee}.Value)
   RigVMFunction_Sequence = &RigVMFunction_Sequence(ExecuteContext=@n5.Completed)
   n6 = RigVMVariableNode() {variable=Initial Duration}
   n7 = /Script/RigVM.RigVMFunction_TimeLoop(Normalize=false, Speed=1, Duration=@n6.Value) {is=struct RigVMUnitNode}
   n8 = &RigVMFunction_ControlFlowBranch(
    Condition=/Script/RigVM.RigVMFunction_MathDoubleGreaterEqual(A=@n7.Relative, B=&RigVMFunction_MathDoubleSub(B=0.1, A=@n6.Value).Result) {is=struct RigVMUnitNode, template="GreaterEqual::Execute(in A,in B,out Result)"}.Result
    ExecuteContext=@RigVMFunction_Sequence.A
   )
   Multiply = &RigVMFunction_MathDoubleMul(B=0.5, A=@n6.Value)
   n9 = &RigVMFunction_ControlFlowBranch(
    Condition=&RigVMFunction_MathDoubleGreaterEqual(B=@Multiply.Result, A=@n7.Relative) {template="GreaterEqual::Execute(in A,in B,out Result)"}.Result
    ExecuteContext=@n8.Completed
   )
   AccumulateLerp = /Script/RigVM.RigVMFunction_AccumulateFloatLerp(
    bIntegrateDeltaTime=true
    Blend=RigVMVariableNode() {variable=blend time}.Value
    InitialValue=RigVMVariableNode() {variable=Next Speed}.Value
    TargetValue=RigVMVariableNode() {variable=Current Speed}.Value
   ) {is=struct RigVMUnitNode, template="AccumulateLerp::Execute(in TargetValue,in InitialValue,in Blend,in bIntegrateDeltaTime,out Result)"}
   RigVMFunction_TimeLoop = &RigVMFunction_TimeLoop(Duration=360, Normalize=false, Speed=@AccumulateLerp.Result)
   RigVMFunction_MathQuaternionFromEuler = &RigVMFunction_MathQuaternionFromEuler(Euler=(X=0,Y=0,Z=0), Euler.Z=@RigVMFunction_TimeLoop.Relative)
   &RigUnit_ModifyTransforms(
    ItemToModify=((Item=(Type=Bone,Name="base_paintRoller"),Transform=(Rotation=(X=0,Y=0,Z=0,W=1),Translation=(X=0,Y=0,Z=0),Scale3D=(X=1,Y=1,Z=1))))
    Weight=1
    WeightMinimum=0
    ExecutePin=@n9.Completed
    ItemToModify.0.Transform.Rotation=@RigVMFunction_MathQuaternionFromEuler.Result
   )
   n10 = RigVMVariableNode() {variable=Speed And Duration}
   RigVMVariableNode(ExecuteContext=@n9.True, Value=@n10.Value.X) {variable=Next Speed}
   RigVMVariableNode(ExecuteContext=@n8.False, Value=RigVMVariableNode() {variable=Initial Duration}.Value) {variable=Initial Duration}
   RigVMVariableNode(ExecuteContext=@n8.True, Value=@n10.Value.Y) {variable=Initial Duration}
   > RigVMVariableNode(Value=@n10.Value.Y) {variable=Current Duration}
   > RigVMVariableNode(Value=@n10.Value.X) {variable=Current Speed}
   Subtract = &RigVMFunction_MathDoubleSub(B=3, A=RigVMVariableNode() {variable=swing duration}.Value)
   RigVMVariableNode(
    Value=RigVMDispatchNode(
     False=0
     Condition=/Script/RigVM.RigVMFunction_MathDoubleLessEqual(A=0, B=@Subtract.Result) {is=struct RigVMUnitNode, template="LessEqual::Execute(in A,in B,out Result)"}.Result
     True=@Subtract.Result
    ) {template="DISPATCH_RigVMDispatch_If(in Condition,in True,in False,out Result)"}.Result
    ExecuteContext=@n5.False
   ) {variable=swing duration}
   Add = &RigVMFunction_MathDoubleAdd(B=0.05, A=RigVMVariableNode() {variable=swing duration}.Value)
   Clamp = &RigVMFunction_MathDoubleClamp(Maximum=2, Value=@Add.Result)
   RigVMVariableNode(Value=@Clamp.Result, ExecuteContext=@n5.True) {variable=swing duration}
   GreaterEqual = &RigVMFunction_MathDoubleGreaterEqual(B=1.75, A=@Clamp.Result) {template="GreaterEqual::Execute(in A,in B,out Result)"}
   > RigVMVariableNode(
    Value=(X=200,Y=2000,Z=0)
    Value.Y=RigVMDispatchNode(False=0.5, True=@Clamp.Result, Condition=@GreaterEqual.Result) {template="DISPATCH_RigVMDispatch_If(in Condition,in True,in False,out Result)"}.Result
   ) {variable=Speed And Duration}
   n11 = &RigVMFunction_MathDoubleGreaterEqual(B=1, A=@Clamp.Result) {template="GreaterEqual::Execute(in A,in B,out Result)"}
   > RigVMVariableNode(Value=RigVMDispatchNode(
     True=1.5
     Condition=@GreaterEqual.Result
     False=RigVMDispatchNode(False=5, True=1, Condition=@n11.Result) {template="DISPATCH_RigVMDispatch_If(in Condition,in True,in False,out Result)"}.Result
    ) {template="DISPATCH_RigVMDispatch_If(in Condition,in True,in False,out Result)"}.Result) {variable=blend time}
   n12 = &RigVMFunction_MathDoubleAdd(B=0.1, A=&RigVMFunction_TimeLoop(Duration=16, Normalize=false, Speed=1).Relative)
   RigVMVariableNode(
    Value=(X=1000,Y=5,Z=0)
    Value.X=RigVMDispatchNode(
     False=0
     Condition=&RigVMFunction_MathDoubleGreaterEqual(B=8, A=@n12.Result) {template="GreaterEqual::Execute(in A,in B,out Result)"}.Result
     True=/Script/RigVM.RigVMFunction_RandomFloat(Duration=10, Maximum=50, Minimum=0, Seed=500) {is=struct RigVMUnitNode, template="Random::Execute(in Seed,in Minimum,in Maximum,in Duration,out Result)"}.Result
    ) {template="DISPATCH_RigVMDispatch_If(in Condition,in True,in False,out Result)"}.Result
    ExecuteContext=@n4.True
   ) {variable=Speed And Duration}
   RigVMVariableNode(Value=(X=100,Y=0,Z=0), ExecuteContext=@n3.True) {variable=Speed And Duration}
   RigVMVariableNode(Value=(X=250,Y=0,Z=0), ExecuteContext=@n2.True) {variable=Speed And Duration}
   RigVMVariableNode(Value=(X=25,Y=0,Z=0), ExecuteContext=@n1.True) {variable=Speed And Duration}
   RigVMVariableNode(Value=(X=0,Y=0,Z=0), ExecuteContext=@RigVMFunction_ControlFlowBranch.True) {variable=Speed And Duration}
   &RigVMFunction_ControlFlowBranch(Condition=False, ExecuteContext=())
   &RigVMFunction_MathDoubleSub(B=0.5, A=@n6.Value)
  ]}
]
hierarchy [
 b_AR_Root {loc=<0 0 0>, rot=<0 0 0.7071 0.7071>}
 Melee_TwoHanded_Handle {parent=b_AR_Root, loc=<8.1477 -0.008 -29.0793>, rot=<0 0 0.7071 -0.7071>}
 base_paintRoller {parent=Melee_TwoHanded_Handle, loc=<0.7027 -0.9913 96.9553>, rot=<0.7071 0 0 -0.7071>}
 dyn_paintRoller {parent=base_paintRoller, loc=<0 0 0>, rot=<0 0 0 1>}
 base_rollercap {parent=dyn_paintRoller, loc=<0.0388 0.0359 -25.8689>, rot=<-0.7044 0.0616 0.0616 -0.7044>}
 base_cap {parent=base_rollercap, loc=<8.1337 -1.3325 0.2153>, rot=<0 0 0 1>}
 dyn_rollerCap {parent=dyn_paintRoller, loc=<3.8951 -0.5473 -26.9797>, rot=<-0.7044 0.0616 0.0616 -0.7044>}
]
variables [
 Current Duration {type=double, default=0}
 Current Speed {type=double, default=0}
 Initial Duration {type=double, default=0}
 Is Falling {type=bool, default=False}
 Is Idle {type=bool, default=False}
 Is Melee {type=bool, default=False}
 Is Zipplining {type=bool, default=False}
 IsFrontEnd {type=bool, default=False}
 IsMoving2D {type=bool, default=False}
 Next Speed {type=double, default=0}
 Speed And Duration {type=FVector, default=(X=0,Y=0,Z=0)}
 blend time {type=double, default=0.5}
 swing duration {type=double, default=0}
]
