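# Animation Blueprint for the VividSplash paint-roller pickaxe, driving the weapon's secondary motion from its state.
# The input pose runs through a Control Rig fed melee, idle, falling, ziplining, 2D movement and front-end flags.
# An AnimDynamics node then adds spring-driven, angle-limited physics to the roller cap bone, simulated relative to the paint roller bone.
# The event graph sets idle as not moving and toggles ziplining from player zipline enter and exit anim input events.
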AnimBlueprint /BRCosmetics/Weapons/FORT_Melee/Pickaxe_VividSplash/Meshes/VividSplash_Axe_AnimBP
target_skeleton=&/VividSplash_Axe_Skeleton
graphs [
 AnimGraph {nodes [
   AnimGraphNode_SubInput_0 = Input Pose()
   n1 = Control Rig!(
    Source=@AnimGraphNode_SubInput_0
    Alpha
    Is Melee:bool=Get bIsPlayingMeleeAnim()
    Is Idle:bool=Get Is IDLE() -> (Is IDLE:bool)
    Is Falling:bool=Get bIsFalling()
    Is Zipplining:bool=Get isZiplining() -> (isZiplining:bool)
    IsMoving2D:bool=Get bIsMoving2D()
    IsFrontEnd:bool=Get bIsInFrontEnd()
   ) -> (Pose) {
    alphaBoolBlend {blendOption=Linear}
    controlRigAssetReference {blueprintRigClass=&/VividSplash_Axe_CtrlRig.VividSplash_Axe_CtrlRig_C}
   }
   n2 = Anim Dynamics - Bone: dyn_rollerCap(ComponentPose=Local To Component(LocalPose=@n1), Alpha=1, ExternalForce="0,0,0") {
    angularDampingOverride=1
    angularSpringConstant=20000
    bAngularSpring=true
    bOverrideAngularDamping=true
    bUseGravityOverride=true
    gravityScale=0.35
    simulationSpace=BoneRelative
    alphaBoolBlend {blendOption=Linear}
    boundBone {boneName=dyn_rollerCap}
    componentAppliedLinearAccClamp=<3000 3000 3000>
    componentLinearAccScale {y=1, z=1}
    componentLinearVelScale {y=1, z=1}
    gravityOverride {z=-980}
    physicsBodyDefinitions [{
      boundBone {boneName=dyn_rollerCap}
      boxExtents=<30 30 30>
      constraintSetup {angularLimitsMax=<0 30 20>, angularLimitsMin=<0 -20 -20>, angularTarget=<1 0 0>}
      localJointOffset=<4 0 0>
     }]
    relativeSpaceBone {boneName=dyn_paintRoller}
   }
   Output Pose(Result=Component To Local(ComponentPose=@n2)) {layerGroup=None}
  ]}
 EventGraph {nodes [
   n1 = Event BlueprintUpdateAnimation()
   > Set DeltaTime(DeltaTime:double=@n1.DeltaTimeX) -> (Output_Get:double)
   > Set Is IDLE(Is IDLE:bool=NOT Boolean(A=Get bIsMoving2D())) -> (Output_Get:bool)
   n2 = Event BlueprintProcessAnimInputEvent()
   > n3 = Branch(Condition=IsNamed(self=@n2.AnimInputEvent, InEventName=aie.Player.Zipline.Entered))
   > Set isZiplining(isZiplining:bool=true) -> (Output_Get:bool)
   Branch(execute=@n3.else, Condition=IsNamed(self=@n2.AnimInputEvent, InEventName=aie.Player.Zipline.Exited))
   > Set isZiplining(isZiplining=false)
   TryGetPawnOwner()
  ]}
]
variables [
 DeltaTime {type=number, default=0.0106}
 Is IDLE {type=boolean, default=true}
 isZiplining {type=boolean}
]
